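# Niagara glow effect for the LemonWrath Ochre backpack cosmetic, built from a single MainGlow sprite emitter.
# A burst spawns persistent sprites with fixed facing, a camera offset, and a position offset that differs between front end and gameplay.
# User.VariantColor indexes a two-entry color array, selecting orange or fully transparent black.
# Alpha scale depends on both platform and front-end versus gameplay context.
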NiagaraSystem /BRCosmetics/Accessories/FORT_Backpacks/Backpack_F_MED_LemonWrath_Ochre/FX/NS_Backpack_LemonWrath_Ochre
dependencies {
 usedDataInterfaces [&NiagaraDataInterfaceArrayColor, &NiagaraDataInterfacePlatformSet]
 usedDynamicInputs []
 usedModules [&SystemState]
}
emitters {MainGlow {
  emitterUpdateScript [
   &EmitterState(
    Inactive Response=::NewEnumerator1
    Life Cycle Mode=::NewEnumerator1
    Loop Behavior=::NewEnumerator1
    Loop Duration Mode=::NewEnumerator1
    UseLoopDelay=false
   )
   &SpawnBurst_Instantaneous(Use Spawn Probability=false, UseLoopCountLimit=false)
  ]
  particleSpawnScript [
   &InitializeParticle(
    Color Mode=::NewEnumerator0
    Position Offset=&FN_FrontEndVector3Modifier(Gameplay=<-1 -9.8 -1.5>, Front End=<-1 -9 -2>)
    Sprite Size Mode=::NewEnumerator3
    Uniform Sprite Size=20
    UsePositionOffset=true
    Write Material Random=true
   )
   &DynamicMaterialParameters(
    Index 0 Param 1=5
    Index 0 Param 2=5
    Module.Index0Float0Write=true
    Module.Index0Float1Write=true
    Module.Index0Float2Write=true
    Module.Index0Float3Write=true
   )
   &CameraOffset(Camera Offset Amount=5)
   &SpriteFacingAndAlignment(Sprite Alignment=<0 0 1>, Sprite Facing=<0 -1 0>, WriteAlignment=true, WriteFacing=true)
  ]
  particleUpdateScript [
   &ParticleState(Kill Particles When Lifetime Has Elapsed=false, Loop Particles Lifetime=true)
   &Color(Color=&SelectLinearColorFromArray(
     Array Sampling Mode=::NewEnumerator1
     Direct Array Index=&MakeIntFromFloat(Float=$User.VariantColor)
     Color Selection Array=DataInterface() {propertyValues {ColorData [<rgba 1 0.3333 0 1>, <rgba 0 0 0 0>]}}
     Random Seed=?
    ))
   &ScaleColor(
    Color Value To Scale=$Particles.Color
    Scale Alpha=&FN_PickFloatDependOnPlatform(
     TargetPlatformActive=&FN_FrontEndFloatModifier(Gameplay=2, Front End=1.5)
     TargetPlatformNoActive=&FN_FrontEndFloatModifier(Gameplay=0.2, Front End=0.1)
    )
    ScaleA=true
    ScaleRGB=true
   )
  ]
  rendererClasses [&NiagaraSpriteRendererProperties]
  renderers [{rendererClass=&NiagaraSpriteRendererProperties}]
 }}
summary {
 systemName=NS_Backpack_LemonWrath_Ochre
 emitters [{emitterName=MainGlow, rendererClasses [&NiagaraSpriteRendererProperties]}]
 userVariables [
  User.bIsFrontEnd {type=NiagaraBool}
  User.bIsFrontEndPreview {type=NiagaraBool}
  User.VariantColor {type=NiagaraFloat, defaultValue=0, description=""}
 ]
}
