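# Physics asset for the LemonWrath Ochre backpack cosmetic, defining capsule bodies and joints for the bag, its keychains and its straps.
# A keychain base hangs off the bag and branches into two two-link chains with mostly wide swing and minimal twist.
# Left and right four-link straps hang from the bag with tight angular limits.
# A tightly limited constraint joins the two strap ends, closing them into a loop.
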PhysicsAsset /BRCosmetics/Accessories/FORT_Backpacks/Backpack_F_MED_LemonWrath_Ochre/Meshes/F_MED_LemonWrath_Ochre_Pack_Physics
bodies {
 base_bag_01 [
  {shapeType=Capsule, length=10.2301, radius=4.954, center=<2.1361 1.1961 -10.7268>, rotation=<rot -2.6104 -88.2083 166.3863>}
  {
   shapeType=Capsule
   length=3.5
   radius=5
   shapeName=base_bag_01_capsule
   center=<-0.8992 -1.004 -0.1778>
   rotation=<rot 24.1861 -0.0033 -0.0083>
  }
 ]
 dyn_keyChainA_01 [{shapeType=Capsule, length=0.5, radius=0.5, shapeName=dyn_keyChainA_01_capsule, rotation=<rot 90 0 0>}]
 dyn_keyChainA_02 [{
   shapeType=Capsule
   length=1
   radius=1.5
   shapeName=dyn_keyChainA_02_capsule
   center=<2.1383 0.2444 -0.2496>
   rotation=<rot -67.5278 108.6523 -110.4644>
  }]
 dyn_keyChainB_01 [{shapeType=Capsule, length=0.5, radius=0.5, shapeName=dyn_keyChainB_01_capsule, rotation=<rot 90 0 0>}]
 dyn_keyChainB_02 [{
   shapeType=Capsule
   length=0
   radius=1.5
   shapeName=dyn_keyChainB_02_capsule
   center=<1.6532 0.2502 -0.5775>
   rotation=<rot 11.7522 175.4592 -139.4827>
  }]
 dyn_keyChainBase_01 [{shapeType=Capsule, length=0.5, radius=0.5, shapeName=dyn_keyChainBase_01_capsule, rotation=<rot 90 0 0>}]
 dyn_strap_01_l [{shapeType=Capsule, length=2, radius=1, shapeName=dyn_strap_01_l_capsule, center=<2 0 0>, rotation=<rot 90 0 0>}]
 dyn_strap_01_r [{shapeType=Capsule, length=2, radius=1, shapeName=dyn_strap_01_r_capsule, center=<2 0 0>, rotation=<rot 90 0 0>}]
 dyn_strap_02_l [{shapeType=Capsule, length=1, radius=2.5, shapeName=dyn_strap_02_l_capsule, center=<2.22 0 -0.5433>, rotation=<rot 90 0 0>}]
 dyn_strap_02_r [{shapeType=Capsule, length=2, radius=1, shapeName=dyn_strap_02_r_capsule, center=<2 0 0>, rotation=<rot 90 0 0>}]
 dyn_strap_03_l [{shapeType=Capsule, length=0.5, radius=2.5, shapeName=dyn_strap_03_l_capsule, center=<1.3273 0.0183 0>, rotation=<rot 90 0 0>}]
 dyn_strap_03_r [{shapeType=Capsule, length=2, radius=1, shapeName=dyn_strap_03_r_capsule, center=<2 0 0>, rotation=<rot 90 0 0>}]
 dyn_strap_04_l [{shapeType=Capsule, length=1, radius=2, shapeName=dyn_strap_04_l_capsule, center=<0 0.51 0>, rotation=<rot 90 0 0>}]
 dyn_strap_04_r [{shapeType=Capsule, length=1, radius=2, shapeName=dyn_strap_04_r_capsule, rotation=<rot 90 0 0>}]
}
constraints [
 {bone1Name=dyn_keyChainBase_01, bone2Name=base_bag_01, swing1LimitDegrees=50, swing2LimitDegrees=20, twistLimitDegrees=1}
 {bone1Name=dyn_keyChainA_01, bone2Name=dyn_keyChainBase_01, swing1LimitDegrees=60.56, swing2LimitDegrees=20, twistLimitDegrees=1}
 {bone1Name=dyn_keyChainA_02, bone2Name=dyn_keyChainA_01, swing1LimitDegrees=75, swing2LimitDegrees=10, twistLimitDegrees=1}
 {bone1Name=dyn_keyChainB_01, bone2Name=dyn_keyChainBase_01, swing1LimitDegrees=75, swing2LimitDegrees=45, twistLimitDegrees=1}
 {bone1Name=dyn_keyChainB_02, bone2Name=dyn_keyChainB_01, swing1LimitDegrees=10, swing2LimitDegrees=75, twistLimitDegrees=1}
 {bone1Name=dyn_strap_01_l, bone2Name=base_bag_01, swing1LimitDegrees=1, swing2LimitDegrees=1, twistLimitDegrees=1}
 {bone1Name=dyn_strap_02_l, bone2Name=dyn_strap_01_l, swing1LimitDegrees=5, swing2LimitDegrees=5, twistLimitDegrees=10}
 {bone1Name=dyn_strap_03_l, bone2Name=dyn_strap_02_l, swing1LimitDegrees=15, swing2LimitDegrees=1, twistLimitDegrees=10}
 {bone1Name=dyn_strap_04_l, bone2Name=dyn_strap_03_l, swing1LimitDegrees=15, swing2LimitDegrees=1, twistLimitDegrees=15}
 {bone1Name=dyn_strap_01_r, bone2Name=base_bag_01, swing1LimitDegrees=1, swing2LimitDegrees=1, twistLimitDegrees=1}
 {bone1Name=dyn_strap_02_r, bone2Name=dyn_strap_01_r, swing1LimitDegrees=1, swing2LimitDegrees=10, twistLimitDegrees=10}
 {bone1Name=dyn_strap_03_r, bone2Name=dyn_strap_02_r, swing1LimitDegrees=15, swing2LimitDegrees=1, twistLimitDegrees=10}
 {bone1Name=dyn_strap_04_r, bone2Name=dyn_strap_03_r, swing1LimitDegrees=15, swing2LimitDegrees=1, twistLimitDegrees=10}
 {bone1Name=dyn_strap_04_l, bone2Name=dyn_strap_04_r, swing1LimitDegrees=1, swing2LimitDegrees=1, twistLimitDegrees=1}
]
