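# Animation Blueprint for the LemonWrath Ochre backpack, adding physics motion and corrective offsets to the pack bone.
# The event graph maps the owner mesh's spine_05-to-spine_01 yaw delta from ref pose into backpack rotation and translation.
# Both branches apply a fixed offset and RigidBody simulation in pack-bone space with custom gravity.
# The front-end branch stops there, while the in-game branch adds the spine-driven offset before simulating with different sim-space settings.
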AnimBlueprint /BRCosmetics/Accessories/FORT_Backpacks/Backpack_F_MED_LemonWrath_Ochre/Meshes/F_MED_LemonWrath_Ochre_Pack_AnimBP
target_skeleton=&/F_MED_LemonWrath_Ochre_Pack_Skeleton
graphs [
 AnimGraph {nodes [
   n1 = Use cached pose 'Start_Pose'()
   n2 = RigidBody!(
    ComponentPose="Transform (Modify) Bone - Bone: pack"!(ComponentPose=Local To Component(LocalPose=@n1), Alpha, Rotation="0,0,0", Scale) -> (Pose) {rotationMode=BMM_Additive, translationMode=BMM_Additive, boneToModify {boneName=pack}, translation {y=3}}
    Alpha
    OverrideWorldGravity:Vector="0,0,-980"
    ExternalForce
   ) -> (Pose) {
    bClampLinearTranslationLimitToRefPose=true
    bOverrideWorldGravity=true
    overridePhysicsAsset=&/F_MED_LemonWrath_Ochre_Pack_Physics
    simulationSpace=BaseBoneSpace
    baseBoneRef {boneName=pack}
    simSpaceSettings {
     dampingAlpha=0
     maxAngularAcceleration=100
     maxAngularVelocity=20
     maxLinearAcceleration=2800
     maxLinearVelocity=700
     velocityScaleZ=0
     worldAlpha=1
     externalAngularVelocity=<30 30 30>
    }
   }
   n3 = Local To Component(LocalPose=Use cached pose 'Start_Pose'())
   n4 = RigidBody(
    ComponentPose="Transform (Modify) Bone - Bone: pack"!(
     ComponentPose="Transform (Modify) Bone - Bone: pack"!(ComponentPose=@n3, Alpha, Rotation="0,0,0", Scale) -> (Pose) {rotationMode=BMM_Additive, translationMode=BMM_Additive, boneToModify {boneName=pack}, translation {y=3}}
     Alpha
     Translation:Vector=Get Backpack Translation() -> (Backpack Translation:Vector)
     Rotation=Get Backpack Rotate() -> (Backpack Rotate:Rotator)
     Scale
    ) -> (Pose) {rotationMode=BMM_Additive, translationMode=BMM_Additive, boneToModify {boneName=pack}}
    OverrideWorldGravity="0,0,-980"
   ) {
    bClampLinearTranslationLimitToRefPose=true
    bOverrideWorldGravity=true
    overridePhysicsAsset=&F_MED_LemonWrath_Ochre_Pack_Physics
    simulationSpace=BaseBoneSpace
    baseBoneRef {boneName=pack}
    simSpaceSettings {
     dampingAlpha=0
     maxAngularAcceleration=100
     maxAngularVelocity=20
     maxLinearAcceleration=2800
     maxLinearVelocity=700
     velocityScaleZ=0.75
     worldAlpha=1
     externalLinearDragV=<0.2 0.2 0.2>
    }
   }
   Output Pose(Result=Blend Poses by bool(BlendPose_0=Component To Local(ComponentPose=@n2), BlendPose_1=Component To Local(ComponentPose=@n4), bActiveValue=Get bIsInFrontEnd())) {layerGroup=None}
   Save cached pose 'Start_Pose'(Pose=Input Pose()) {cacheName=Start_Pose}
  ]}
 EventGraph {nodes [
   Event BlueprintUpdateAnimation()
   > Sequence()
   n1 = GetOwnerSkeletalMeshComponent()
   n2 = BreakRotator(InRot=BreakTransform(InTransform=GetDeltaTransformFromRefPose(self=@n1, BoneName=spine_05, BaseName=spine_01)).Rotation)
   > Set Backpack Rotate(Backpack Rotate:Rotator=MakeRotator(
     Roll=MapRangeClamped(Value=@n2.Yaw, InRangeA=0, InRangeB=-30, OutRangeA=5, OutRangeB=25)
     Pitch=MapRangeClamped(Value=@n2.Yaw, InRangeA=0, InRangeB=-30, OutRangeA=5, OutRangeB=25)
    )) -> (Output_Get:Rotator)
   > Set Backpack Translation(Backpack Translation:Vector=MakeVector(Y=MapRangeClamped(Value=@n2.Yaw, InRangeA=0, InRangeB=-30, OutRangeB=-6.5))) -> (Output_Get:Vector)
   TryGetPawnOwner()
  ]}
]
variables [
 Backpack Rotate {type=Rotator}
 Backpack Translation {type=Vector}
]
